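# Conversation database for the AshenMagnus NPC, Killswitch of The Supernovas, defining the NPC's greeting and dialogue entry.
# A single speech node picks contextual lines: a first-ever-meeting intro, match conversation-count lines, and higher-priority lines for players wearing specific character cosmetics.
# The speech node then branches through link nodes to the shared NPC quest and basic services entries and the separate AshenMagnus dialogue entry.
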ConversationDatabase /NPCLibrary/NPCs/AshenMagnus/NPCLibrary_Conversation_AshenMagnus
entry_points [{tag=Conversation.Entry.AshenMagnus, destinations [#1]}]
nodes [
 ConversationEntryPointNode_7 {
  id=#1
  class=ConversationEntryPointNode
  outputs [#2]
  properties {entryTag {tagName=Conversation.Entry.AshenMagnus}}
 }
 Speech_C_4 {
  id=#2
  class=Speech_C
  outputs [#3, #4, #5]
  properties {
   nodeName=""
   generalConfig {
    defaultMessage="The world's always gonna try and chew you up, so sharpen your... teeth? Erm..."
    contextualMessages [
     {
      message="Hey! I'm Killswitch, one of The Supernovas."
      contextRequirements [{requirement {class=FortControllerRequirement_IsFirstEverConversationWithNPC}}]
      priority {curve {rowName=Tandem.Speech.Priority.FirstEverInteraction}}
     }
     {
      message="The world's always gonna try and chew you up, so sharpen your... teeth? Erm..."
      contextRequirements [{requirement {
         class=FortControllerRequirement_CheckNPCMatchConversationCount
         properties {nPCUniqueIdTag {tagName=AISpawnerData.Type.Specialist.Scout.AshenMagnus}}
        }}]
      priority {curve {rowName=None}}
     }
     {
      message=I'm not afraid of any bugs. ESPECIALLY not mosquitos.
      contextRequirements [{requirement {
         class=FortControllerRequirement_CheckNPCMatchConversationCount
         properties {nPCUniqueIdTag {tagName=AISpawnerData.Type.Specialist.Scout.AshenMagnus}}
        }}]
      priority {curve {rowName=None}}
     }
     {
      message="The Supernovas have got this handled! And, uh, O.X.R. can be there too, I guess."
      contextRequirements [{requirement {
         class=FortControllerRequirement_CheckNPCMatchConversationCount
         properties {nPCUniqueIdTag {tagName=AISpawnerData.Type.Specialist.Scout.AshenMagnus}}
        }}]
      priority {curve {rowName=None}}
     }
     {
      message="WHEN I YELL \"SUPER\", YOU YELL \"NOVAS\"! YEAH!"
      contextRequirements [{requirement {
         class=FortControllerRequirement_HasCID
         properties {allowedCIds [
           /BRCosmetics/Athena/Items/Cosmetics/Characters/Character_BentBaton
           /BRCosmetics/Athena/Items/Cosmetics/Characters/Character_LazyLizzChip
           /BRCosmetics/Athena/Items/Cosmetics/Characters/Character_SpatialTravelWatt
           /BRCosmetics/Athena/Items/Cosmetics/Characters/Character_PolarityWinnField
           /BRCosmetics/Athena/Items/Cosmetics/Characters/Character_MistRavenPeck
          ]}
        }}]
      priority {value=2, curve {curveTable=null, rowName=None}}
     }
     {
      message="Hey... We should get in a training room and 1v1 sometime, yeah?"
      contextRequirements [{requirement {class=FortControllerRequirement_HasCID, properties {allowedCIds [/BRCosmetics/Athena/Items/Cosmetics/Characters/Character_TalonPane_Flail]}}}]
      priority {value=2, curve {curveTable=null, rowName=None}}
     }
     {
      message=Ma'am.
      contextRequirements [{requirement {class=FortControllerRequirement_HasCID, properties {allowedCIds [&Character_RoseDepth_Seed]}}}]
      priority {value=2, curve {curveTable=null, rowName=None}}
     }
    ]
   }
  }
 }
 ConversationLinkNode_2 {
  id=#5
  class=ConversationLinkNode
  properties {remoteEntryTag {tagName=Conversation.Entry.AshenMagnus.Dialogue}}
 }
 ConversationLinkNode_0 {id=#4, class=ConversationLinkNode, properties {remoteEntryTag {tagName=Conversation.Entry.BasicServices}}}
 ConversationLinkNode_1 {id=#3, class=ConversationLinkNode, properties {remoteEntryTag {tagName=Conversation.Entry.NPCLibrary.Quests}}}
]
