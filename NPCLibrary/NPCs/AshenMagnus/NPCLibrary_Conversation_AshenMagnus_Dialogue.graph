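# Conversation database for the Ashen Magnus NPC's dialogue, giving players lore banter about The Supernovas and their rival O.X.R.
# An entry point tagged Conversation.Entry.AshenMagnus.Dialogue leads to a greeting, shown as an "Ask about..." choice, that branches into two topic choices and a Back option.
# Each topic plays a linear chain of three speech lines, and both chains converge on one shared Return To Start node labeled BACK.
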ConversationDatabase /NPCLibrary/NPCs/AshenMagnus/NPCLibrary_Conversation_AshenMagnus_Dialogue
entry_points [{tag=Conversation.Entry.AshenMagnus.Dialogue, destinations [#1]}]
nodes [
 ConversationEntryPointNode_0 {
  id=#1
  class=ConversationEntryPointNode
  outputs [#2]
  properties {entryTag {tagName=Conversation.Entry.AshenMagnus.Dialogue}}
 }
 Speech_C_0 {
  id=#2
  class=Speech_C
  outputs [#3, #4, #5]
  properties {nodeName="", softTaskIcon=&T_UI_NPC_question, generalConfig {defaultMessage="Hey, what's up?"}}
  sub_nodes [ConversationChoiceNode_5 {class=ConversationChoiceNode, properties {defaultChoiceDisplayText=Ask about...}}]
 }
 Speech_C_1 {
  id=#5
  class=Speech_C
  outputs [#6]
  properties {
   nodeName=""
   softTaskIcon=&T_UI_NPC_question
   generalConfig {defaultMessage="O.X.R. has all their guns and gadgets, but they'll never shape up to The Supernovas."}
  }
  sub_nodes [ConversationChoiceNode_1 {class=ConversationChoiceNode, properties {defaultChoiceDisplayText=O.X.R.}}]
 }
 Speech_C_2 {
  id=#6
  class=Speech_C
  outputs [#7]
  properties {
   nodeName=""
   softTaskIcon=&T_UI_NPC_question
   generalConfig {defaultMessage="Sure, they're led by a world-class tactician with some of the best operatives in the world."}
  }
 }
 Speech_C_4 {
  id=#7
  class=Speech_C
  outputs [#8]
  properties {
   nodeName=Speech
   softTaskIcon=&T_UI_NPC_question
   generalConfig {defaultMessage="But we've got a STORM BEAST! Check and mate."}
  }
 }
 ReturnToStart_C_23 {
  id=#8
  class=ReturnToStart_C
  properties {softTaskIcon=&T_UI_NPC_Back}
  sub_nodes [ConversationChoiceNode_32 {class=ConversationChoiceNode, properties {defaultChoiceDisplayText=BACK}}]
 }
 Back_C_0 {
  id=#4
  class=Back_C
  sub_nodes [ConversationChoiceNode_7 {class=ConversationChoiceNode, properties {defaultChoiceDisplayText=BACK}}]
 }
 Speech_C_7 {
  id=#3
  class=Speech_C
  outputs [#9]
  properties {
   nodeName=""
   softTaskIcon=&T_UI_NPC_question
   generalConfig {defaultMessage="We put the SUPER in SUPERNOVAS. We're the BEST!"}
  }
  sub_nodes [ConversationChoiceNode_0 {class=ConversationChoiceNode, properties {defaultChoiceDisplayText=The Supernovas}}]
 }
 Speech_C_15 {
  id=#9
  class=Speech_C
  outputs [#10]
  properties {
   nodeName=""
   softTaskIcon=&T_UI_NPC_question
   generalConfig {defaultMessage=We went through it to get here. Learned the hard way about being heroes.}
  }
 }
 Speech_C_14 {
  id=#10
  class=Speech_C
  outputs [#8]
  properties {
   nodeName=""
   softTaskIcon=&T_UI_NPC_question
   generalConfig {defaultMessage="But nobody can protect this Island like us. WE'VE GOT THIS! WOO!"}
  }
 }
]
